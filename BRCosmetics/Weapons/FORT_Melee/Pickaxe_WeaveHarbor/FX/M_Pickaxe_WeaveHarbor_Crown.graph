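# Unlit translucent Niagara sprite material for the crown effect on the WeaveHarbor pickaxe.
# A grayscale crown SDF texture minus the CrownErode dynamic parameter feeds a narrow smoothstep, giving a soft-edged shape that particles can erode.
# That mask is multiplied by particle alpha, while particle color passes through MF_FX_EmissiveAdjustments for emissive.
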Material /BRCosmetics/Weapons/FORT_Melee/Pickaxe_WeaveHarbor/FX/M_Pickaxe_WeaveHarbor_Crown
graph {
 properties {
  bAutomaticallySetUsageInEditor=false
  bUsedWithNiagaraSprites=true
  blendMode=BLEND_Translucent
  pixelDepthOffsetMode=PDOM_Legacy
  shadingModel=MSM_Unlit
  usedShadingModels=Unlit
 }
 nodes [
  n1 = ParticleColor()
  MaterialFunctionCall(Emissive=@n1.RGB) {materialFunction=&MF_FX_EmissiveAdjustments}
  n2 = TextureSample() {samplerType=SAMPLERTYPE_LinearGrayscale, texture=&/T_Pickaxe_WeaveHarbor_CrownSDF}
  n3 = DynamicParameter() {defaultValue=<rgba 0 1 1 1>, paramNames [CrownErode,Param2,Param3,Param4]}
  Multiply(A=@n1.A, B=SmoothStep(Value=Subtract(A=@n2.R, B=@n3.CrownErode)) {constMax=0.2})
 ]
}
