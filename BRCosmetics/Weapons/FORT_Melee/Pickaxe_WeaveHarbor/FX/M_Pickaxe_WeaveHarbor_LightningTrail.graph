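# Translucent unlit Niagara material for the lightning trail effect on the WeaveHarbor pickaxe, tinted and faded by particle color.
# The lightning texture scrolls in quantized time steps on UV1, with a dynamic LightningWidth threshold that hardens its shape.
# Optional smoothstep masks fade both U edges, and depth fade softens intersections.
# Scrolling Perlin noise, sharpened and scaled by dynamic CoreBright, adds hot flickers to the emissive before a shared adjustment function.
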Material /BRCosmetics/Weapons/FORT_Melee/Pickaxe_WeaveHarbor/FX/M_Pickaxe_WeaveHarbor_LightningTrail
graph {
 properties {
  bAutomaticallySetUsageInEditor=false
  bUsedWithNiagaraMeshParticles=true
  bUsedWithNiagaraRibbons=true
  bUsedWithNiagaraSprites=true
  blendMode=BLEND_Translucent
  pixelDepthOffsetMode=PDOM_Legacy
  shadingModel=MSM_Unlit
  usedShadingModels=Unlit
 }
 nodes [
  n1 = ParticleColor()
  n2 = DynamicParameter() {defaultValue=<rgba 0.9 0 1 1>, paramNames [LightningWidth, CoreBright, Param3, Param4]}
  n3 = Time()
  n4 = VectorParameter() {parameterName=ScrollSpeed, channelNames {a="", b=TimeSteps, g=ScrollSpeedY, r=ScrollSpeedX}, defaultValue=<rgba -2 0 5 1>}
  n5 = ComponentMask(None=@n4) {g=true, r=true}
  n6 = Floor(None=Multiply(A=Multiply(A=@n3, B=@n5), B=@n4.TimeSteps) {constB=2})
  n7 = TextureSampleParameter2D(UVs=Add(A=TextureCoordinate() {coordinateIndex=1}, B=Divide(A=@n6, B=@n4.TimeSteps) {constB=1})) {
   parameterName=MainTexture
   samplerType=SAMPLERTYPE_LinearGrayscale
   texture=&/T_Pickaxe_WeaveHarbor_Lightning
  }
  n8 = LinearInterpolate(A=@n7.R, B=Step(Y=@n2.LightningWidth, X=@n7.R), Alpha=ScalarParameter() {defaultValue=1, parameterName=UseStepping})
  n9 = ComponentMask(None=TextureCoordinate()) {r=true}
  n10 = SmoothStep(Max=ScalarParameter() {defaultValue=0.4, parameterName=EdgeMaskR}, Value=OneMinus(None=@n9)) {constMax=0.4}
  n11 = SmoothStep(Max=ScalarParameter() {defaultValue=0.2, parameterName=EdgeMaskL}, Value=@n9) {constMax=0.2}
  n12 = LinearInterpolate(
   B=Power(Base=Multiply(A=@n10, B=@n11), Exp=ScalarParameter() {defaultValue=1, parameterName=EdgeContrast})
   Alpha=ScalarParameter() {defaultValue=1, parameterName=Use EdgeMaksing}
  ) {constA=1}
  DepthFade(Opacity=Multiply(A=@n1.A, B=Multiply(A=@n8, B=@n12)), FadeDistance=ScalarParameter() {defaultValue=25, parameterName=DepthFadeDistance})
  n13 = Add(A=Multiply(A=TextureCoordinate(), B=Constant2Vector() {r=1}), B=Time())
  n14 = Multiply(A=Power(Base=TextureSample(UVs=@n13) {samplerType=SAMPLERTYPE_LinearGrayscale, texture=&T_SimplePerlin_Greyscale}.R) {constExponent=5}) {constB=15}
  MaterialFunctionCall(Emissive=Add(A=@n1.RGB, B=Multiply(A=@n14, B=@n2.CoreBright))) {materialFunction=&MF_FX_EmissiveAdjustments}
 ]
}
