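# Unlit translucent lightning material for WeaveHarbor pickaxe Niagara mesh and ribbon particles.
# A randomly offset texture, scrolled in time-quantized jumps and multiplied by a panning mask, is thresholded into core and border, or left soft when IsGlow is set.
# Color blends a curve texture or gradient with particle color; opacity adds view-angle and end fades.
# SplineThicken drives world position offset, with dynamic parameters controlling pan, width and core intensity.
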Material /BRCosmetics/Weapons/FORT_Melee/Pickaxe_WeaveHarbor/FX/M_Pickaxe_WeaveHarbor_MeshLightning
graph {
 properties {
  bAutomaticallySetUsageInEditor=false
  bUsedWithNiagaraMeshParticles=true
  bUsedWithNiagaraRibbons=true
  blendMode=BLEND_Translucent
  pixelDepthOffsetMode=PDOM_Legacy
  shadingModel=MSM_Unlit
  twoSided=true
  usedShadingModels=Unlit
 }
 nodes [
  LinearInterpolate(B=NamedRerouteUsage(), Alpha=ScalarParameter() {defaultValue=1, parameterName=UseWPO})
  n1 = ComponentMask(None=TextureCoordinate()) {g=true}
  n2 = ScalarParameter() {defaultValue=1.5, parameterName=CurvePower}
  n3 = Saturate(None=Multiply(A=Power(Base=@n1, Exp=@n2), B=ScalarParameter() {defaultValue=1, parameterName=CurveMultiply}))
  n4 = VectorParameter() {parameterName=TopColor, defaultValue=<rgba 0.6042 0.0123 0.036 1>}
  n5 = VectorParameter() {parameterName=BottomColor, defaultValue=<rgba 0.1227 0 0.4688 1>}
  n6 = SmoothStep(Value=@n1) {constMax=0.9, constMin=0.06}
  n7 = ParticleColor()
  n8 = LinearInterpolate(
   A=StaticSwitchParameter(
    True=TextureSample(UVs=@n3) {texture=&/T_Pickaxe_WeaveHarbor_Color}.RGB
    False=LinearInterpolate(A=@n4.RGB, B=@n5.RGB, Alpha=@n6)
   ) {defaultValue=true, parameterName=UseColorCurve}
   B=@n7.RGB
   Alpha=ScalarParameter() {defaultValue=1, parameterName=UseParticleColor}
  )
  n9 = Multiply(
   A=VectorParameter() {parameterName=CoreColor, defaultValue=<rgba 2 2 2 1>}.RGB
   B=DynamicParameter() {defaultValue=<rgba 1 0 1 1>, paramNames [LightningPan, LightningWidth, LightningIntensity, Param4]}.LightningIntensity
  )
  n10 = TextureCoordinate()
  n11 = MaterialFunctionCall(XY=@n10) {materialFunction=&Swizzle}
  n12 = VectorParameter() {parameterName=TextureScrollSpeed, defaultValue=<rgba 0 0 0 1>}
  n13 = Multiply(A=Multiply(A=Time(), B=ComponentMask(None=@n12.RGB) {g=true, r=true})) {constB=5}
  n14 = Add(A=AppendVector(A=ParticleRandom(), B=Constant()), B=Divide(A=Floor(None=@n13)) {constB=5})
  n15 = VectorParameter() {parameterName=LightningTextureSize, defaultValue=<rgba 0.7 1 0 1>}
  n16 = TextureSampleParameter2D(UVs=Multiply(A=Add(A=@n11.YX, B=@n14), B=ComponentMask(None=@n15.RGB) {g=true, r=true})) {
   parameterName=LightningTexture
   samplerType=SAMPLERTYPE_LinearGrayscale
   texture=&/T_Pickaxe_WeaveHarbor_Lightning
  }
  n17 = LinearInterpolate(Alpha=DynamicParameter() {defaultValue=<rgba 1 0 1 1>, paramNames [LightningPan, LightningWidth, LightningIntensity, Param4]}.LightningPan) {constA=-1}
  n18 = Subtract(A=Add(A=ComponentMask(None=@n10) {g=true}, B=@n17) {constB=0}) {constB=0.5}
  n19 = OneMinus(None=Multiply(A=Abs(None=@n18)) {constB=2})
  n20 = Add(A=ComponentMask(None=TextureCoordinate()) {g=true}, B=@n17)
  n21 = LinearInterpolate(
   B=LinearInterpolate(A=SmoothStep(Value=@n19) {constMax=0.7, constMin=0.4}, B=Saturate(None=@n20), Alpha=ScalarParameter() {parameterName=PanMethod})
   Alpha=ScalarParameter() {defaultValue=1, parameterName=MaskLightning}
  ) {constA=1}
  n22 = Multiply(A=@n16.R, B=@n21)
  n23 = Step(Y=ScalarParameter() {defaultValue=0.8, parameterName=CoreSize}, X=@n22) {constY=0.8}
  MaterialFunctionCall(Emissive=StaticSwitchParameter(True=@n8, False=LinearInterpolate(A=@n8, B=@n9, Alpha=@n23)) {parameterName=IsGlow}) {materialFunction=&MF_FX_EmissiveAdjustments}
  n24 = ScalarParameter() {defaultValue=0.3, parameterName=BorderSize}
  n25 = StaticSwitchParameter(True=@n22, False=Saturate(None=Step(Y=@n24, X=@n22) {constY=0.3})) {parameterName=IsGlow}
  n26 = Multiply(A=DotProduct(A=NamedRerouteUsage(), B=CameraVectorWS())) {constB=1.5}
  n27 = ScalarParameter() {parameterName=MaskTop}
  n28 = Add(A=@n27) {constB=0.4}
  n29 = ComponentMask(None=TextureCoordinate()) {g=true}
  n30 = OneMinus(None=@n29)
  n31 = ScalarParameter() {parameterName=MaskBottom}
  n32 = Add(A=@n31) {constB=0.2}
  Multiply(
   A=Multiply(A=Multiply(A=@n7.A, B=@n25), B=Saturate(None=@n26))
   B=Multiply(A=SmoothStep(Min=@n27, Max=@n28, Value=@n30) {constMax=0.4}, B=SmoothStep(Min=@n31, Max=@n32, Value=@n29) {constMax=0.2})
  )
  n33 = MaterialFunctionCall(WidthBase=ScalarParameter() {parameterName=WidthBase}, WidthTip=ScalarParameter() {parameterName=WidthTip}) {materialFunction=&SplineThicken}
  n34 = Multiply(
   A=DynamicParameter() {defaultValue=<rgba 1 0 1 1>, paramNames [LightningPan, LightningWidth, LightningIntensity, Param4]}.LightningWidth
   B=@n21
  )
  n35 = MaterialFunctionCall(WidthBase=@n34, WidthTip=@n34) {materialFunction=&SplineThicken}
  n36 = ScalarParameter() {defaultValue=1, parameterName=PerParticleSplineTichness}
  NamedRerouteDeclaration(None=LinearInterpolate(A=@n33.WorldPositionOffset, B=@n35.WorldPositionOffset, Alpha=@n36)) {name=STWPO, nodeColor=<rgba 0 0.6719 1 1>}
  NamedRerouteDeclaration(None=LinearInterpolate(A=@n33.Normal, B=@n35.Normal, Alpha=@n36)) {name=STN, nodeColor=<rgba 0.9922 1 0 1>}
  Time() {bOverride_Period=true, period=1}
 ]
}
