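# Niagara swing trail for the WeaveHarbor pickaxe, drawing layered ribbons that follow the weapon's skeletal mesh.
# Four emitters use FN anim-trail modules: a core Trail, a wide faint "Eeeeeeh" layer, and looping Main and Secondary lightning, each with multiple ribbon renderers.
# Curves fade colors, mostly from hot red to violet, and animate lightning material parameters.
# A sprite Smokes emitter that spawns by distance is disabled.
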NiagaraSystem /BRCosmetics/Weapons/FORT_Melee/Pickaxe_WeaveHarbor/FX/NS_Pickaxe_WeaveHarbor_Trail
dependencies {
 usedDataInterfaces [
  &NiagaraDataInterfaceColorCurve
  &NiagaraDataInterfaceCurve
  &NiagaraDataInterfaceSpriteRendererInfo
 ]
 usedDynamicInputs []
 usedModules [
  &self:SystemSpawnScript.SystemScriptSource.SystemScriptGraph.NiagaraNodeAssignment_0.SetVariables_#1
  &self:SystemSpawnScript.SystemScriptSource.SystemScriptGraph.NiagaraNodeAssignment_1.SetVariables_#2
  &SystemState
 ]
}
emitters {
 Eeeeeeh {
  emitterSpawnScript [&FN_RibbonIDOffsetUtility()]
  emitterUpdateScript [
   &EmitterState()
   &FN_SpawnAnimTrail()
  ]
  particleSpawnScript [
   &FN_InitializeAnimTrailBoneIndecies()
   &FN_InitializeAnimTrail(AnimTrailWidth=4, Color=<rgba 1 1 1 0.06>, Lifetime=0.6, SkeletalMesh=$System.Skeletal Mesh)
   &AddVelocity(Velocity=$System.DEBUGVELSPEED) {enabled=false}
  ]
  particleUpdateScript [
   &SolveForcesAndVelocity(Clamp Velocity=false, Limit Acceleration=false)
   &DynamicMaterialParameters(
    Index 0 Param 1=1
    Index 0 Param 2=0.2
    Index 0 Param 3=1
    Index 0 Param 4=1
    Module.Param0WriteEnabled=true
    Module.Param1WriteEnabled=true
    Module.Param2WriteEnabled=true
    Module.Param3WriteEnabled=true
   ) {enabled=false}
   &ScaleColor(
    Linear Color Curve=DataInterface() {propertyValues {
      AlphaCurve {keys [<at 0.0027 1>,<at 0.4857 0>]}
      BlueCurve {keys [<at 0 1.716>,<at 0.0551 0.1274>,<at 0.1569 0>,<at 0.2857 0.8958>]}
      GreenCurve {keys [<at 0 1.7042>,<at 0.0551 0.0166>,<at 0.1569 0>,<at 0.2857 0.1236>]}
      RedCurve {keys [<at 0 2>,<at 0.0551 0.8958>,<at 0.1569 0.3333>,<at 0.2857 0.2381>]}
     }}
    Scale Mode=::NewEnumerator2
   )
   &FN_UpdateAnimTrail(Locked Particle RibbonWidth=1, SkeletalMesh=$System.Skeletal Mesh) {module=FN_UpdateAnimTrail001}
  ]
  rendererClasses [&NiagaraRibbonRendererProperties]
  renderers [
   {rendererClass=&NiagaraRibbonRendererProperties}
   {rendererClass=&NiagaraRibbonRendererProperties, rendererIndex=1}
  ]
 }
 MainLightning {
  emitterSpawnScript [&FN_RibbonIDOffsetUtility()]
  emitterUpdateScript [
   &EmitterState(Life Cycle Mode=::NewEnumerator1, Loop Behavior=::NewEnumerator1, Loop Delay=0.15, Loop Duration=1, UseLoopDelay=true)
   &FN_SpawnAnimTrail()
  ]
  particleSpawnScript [
   &FN_InitializeAnimTrailBoneIndecies()
   &FN_InitializeAnimTrail(AnimTrailWidth=1.3, Color=<rgba 1 1 1 0.8>, Lifetime=0.35, SkeletalMesh=$System.Skeletal Mesh)
   &AddVelocity(Velocity=$System.DEBUGVELSPEED) {enabled=false}
   /SetVariables() {isSetParameters=true, module=SetVariables_#3}
  ]
  particleUpdateScript [
   &SolveForcesAndVelocity(Clamp Velocity=false, Limit Acceleration=false)
   &ScaleColor(
    Linear Color Curve=DataInterface() {propertyValues {
      AlphaCurve {keys [<at 0 1>,<at 0.2742 0>]}
      BlueCurve {keys [<at 0 2>,<at 0.3195 1>]}
      GreenCurve {keys [<at 0 2>,<at 0.3195 1>]}
      RedCurve {keys [<at 0 2>,<at 0.3195 1>]}
     }}
    Scale Mode=::NewEnumerator2
   )
   &DynamicMaterialParameters(
    Index 0 Param 1=&FloatFromCurve(FloatCurve=DataInterface() {propertyValues {Curve {keys [
         {interpMode=RCIM_Cubic, leaveTangent=0.4795, tangentMode=RCTM_Break, time=0, value=0.85}
         {interpMode=RCIM_Cubic, tangentMode=RCTM_Break, time=0.6, value=0.95}
        ]}}})
    Index 0 Param 2=&FloatFromCurve(FloatCurve=DataInterface() {propertyValues {Curve {keys [
         {arriveTangent=9.0725, interpMode=RCIM_Cubic, leaveTangent=9.0725, tangentMode=RCTM_User, time=0, value=0}
         {arriveTangent=-2.5249, interpMode=RCIM_Cubic, leaveTangent=-19.203, tangentMode=RCTM_Break, time=0.1408, value=0.9847}
         {arriveTangent=-0.0405, interpMode=RCIM_Cubic, leaveTangent=-0.0405, tangentMode=RCTM_User, time=0.3, value=0}
        ]}}})
    Module.Index0Param0Write=true
    Module.Index0Param1Write=true
    Module.Index0Param2Write=true
    Module.Index0Param3Write=true
   ) {module=DynamicMaterialParameters001}
   /SetVariables(
    Particles.Dy2:Vector4f=&MakeVector4(
     X=&FloatFromCurve(FloatCurve=DataInterface() {propertyValues {Curve {keys [
          {interpMode=RCIM_Cubic, leaveTangent=0.923, tangentMode=RCTM_Break, time=0, value=0.525}
          {interpMode=RCIM_Cubic, tangentMode=RCTM_Break, time=0.6, value=0.8}
         ]}}})
     Y=0
     Z=0
     W=0
    )
    Particles.Linear Color=&MultiplyLinearColor(
     Linear Color=$Particles.Initial.Color
     Scale Factor=&MakeVector4_FromLinearColor(LinearColor=&ColorFromCurve(ColorCurve=DataInterface() {propertyValues {
         AlphaCurve {keys [<at 0 1>,<at 0.4546 0>]}
         BlueCurve {keys [<at 0 0.0595>,<at 0.1007 0.0086>,<at 0.225 0.8958>]}
         GreenCurve {keys [<at 0 0.0203>,<at 0.1007 0>,<at 0.225 0>]}
         RedCurve {keys [<at 0 1>,<at 0.1007 0.2232>,<at 0.225 0.2341>]}
        }}))
    )
   ) {isSetParameters=true, module=SetVariables_#4}
   &FN_UpdateAnimTrail(Locked Particle RibbonWidth=1, SkeletalMesh=$User.SkeletalMesh) {module=FN_UpdateAnimTrail001}
  ]
  rendererClasses [&NiagaraRibbonRendererProperties]
  renderers [
   {rendererClass=&NiagaraRibbonRendererProperties}
   {rendererClass=&NiagaraRibbonRendererProperties, rendererIndex=1}
   {rendererClass=&NiagaraRibbonRendererProperties, rendererIndex=2}
  ]
 }
 SecondaryLightning {
  emitterSpawnScript [&FN_RibbonIDOffsetUtility()]
  emitterUpdateScript [
   &EmitterState(Life Cycle Mode=::NewEnumerator1, Loop Behavior=::NewEnumerator1, Loop Delay=0.15, Loop Duration=1, UseLoopDelay=true)
   &FN_SpawnAnimTrail()
  ]
  particleSpawnScript [
   &FN_InitializeAnimTrailBoneIndecies()
   &FN_InitializeAnimTrail(AnimTrailWidth=1.2, Color=<rgba 1 1 1 0.8>, Lifetime=0.35, SkeletalMesh=$System.Skeletal Mesh)
   &AddVelocity(Velocity=$System.DEBUGVELSPEED) {enabled=false}
   /SetVariables() {isSetParameters=true, module=SetVariables_#3}
  ]
  particleUpdateScript [
   &SolveForcesAndVelocity(Clamp Velocity=false, Limit Acceleration=false)
   &ScaleColor(
    Linear Color Curve=DataInterface() {propertyValues {
      AlphaCurve {keys [<at 0 1>,<at 0.2762 0>]}
      BlueCurve {keys [<at 0 0.0188>,<at 0.3176 0.0771>]}
      GreenCurve {keys [<at 0 0.0188>,<at 0.3176 0.0059>]}
      RedCurve {keys [<at 0 0.0188>,<at 0.3176 0.0371>]}
     }}
    Scale Mode=::NewEnumerator2
   )
   &DynamicMaterialParameters(
    Index 0 Param 1=&FloatFromCurve(FloatCurve=DataInterface() {propertyValues {Curve {keys [
         {interpMode=RCIM_Cubic, leaveTangent=0.2062, tangentMode=RCTM_Break, time=0, value=0.9}
         {interpMode=RCIM_Cubic, tangentMode=RCTM_Break, time=0.2, value=1}
        ]}}})
    Module.Index0Param0Write=true
    Module.Index0Param1Write=true
    Module.Index0Param2Write=true
    Module.Index0Param3Write=true
   ) {module=DynamicMaterialParameters001}
   /SetVariables(
    Particles.Dy2:Vector4f=&MakeVector4(
     X=&FloatFromCurve(FloatCurve=DataInterface() {propertyValues {Curve {keys [
          {interpMode=RCIM_Cubic, leaveTangent=0.923, tangentMode=RCTM_Break, time=0, value=0.8}
          {interpMode=RCIM_Cubic, tangentMode=RCTM_Break, time=0.6, value=1}
         ]}}})
     Y=0
     Z=0
     W=0
    )
    Particles.Linear Color=&MultiplyLinearColor(
     Linear Color=$Particles.Initial.Color
     Scale Factor=&MakeVector4_FromLinearColor(LinearColor=&ColorFromCurve(ColorCurve=DataInterface() {propertyValues {
         AlphaCurve {keys [<at 0 1>,<at 0.4546 0>]}
         BlueCurve {keys [<at 0 0.0262>,<at 0.0877 0.0019>,<at 0.1653 0.2521>]}
         GreenCurve {keys [<at 0 0.0089>,<at 0.0877 0>,<at 0.1653 0>]}
         RedCurve {keys [<at 0 0.4396>,<at 0.0877 0.1927>,<at 0.1653 0.0659>]}
        }}))
    )
   ) {isSetParameters=true, module=SetVariables_#4}
   &FN_UpdateAnimTrail(Locked Particle RibbonWidth=1, SkeletalMesh=$User.SkeletalMesh) {module=FN_UpdateAnimTrail001}
  ]
  rendererClasses [&NiagaraRibbonRendererProperties]
  renderers [
   {rendererClass=&NiagaraRibbonRendererProperties}
   {rendererClass=&NiagaraRibbonRendererProperties, rendererIndex=1}
   {rendererClass=&NiagaraRibbonRendererProperties, rendererIndex=2}
  ]
 }
 Smokes {
  bEnabled=false
  emitterUpdateScript [
   &EmitterState()
   &SpawnPerUnit(Module.Use Max Movement Threshold=true, Module.Use Movement Tolerance=false, Spawn Spacing=35, Use Spawn Probability=false)
  ]
  particleSpawnScript [
   &InitializeParticle(
    Color=<rgba 0.3308 0.214 0.5938 0.4>
    Lifetime Max=0.3
    Lifetime Min=0.15
    Lifetime Mode=::NewEnumerator1
    Sprite Rotation Mode=::NewEnumerator2
    Sprite Size Mode=::NewEnumerator1
    Uniform Sprite Size Max=80
    Uniform Sprite Size Min=45
    UsePositionOffset=false
    Write Material Random=true
   )
   &SubUVAnimation(
    "Sprite Renderer "=DataInterface() {propertyValues {SpriteRenderer=&self:Smokes_8.NiagaraSpriteRendererProperties_0}}
    SubUV Animation Mode:ENiagara_SubUVLookupModeV2=::NewEnumerator1
    UseEndFrame=false
    UseStartFrame=false
   )
  ]
  particleUpdateScript [
   &ParticleState()
   &InheritVelocity(
    Inherited Velocity Amount Scale=&Multiply_VectorByFloat(Vector=<1 1 1>, Float=-0.2)
    Inherited Velocity Speed Limit=35
    Speed Limit=true
    UseSpeedThreshold=false
   ) {enabled=false}
   &Drag(Drag=2, Use Linear Drag=true, Use Rotational Drag=true) {enabled=false}
   &SolveForcesAndVelocity(Clamp Velocity=false, Limit Acceleration=false)
   &DynamicMaterialParameters(
    Index 0 Param 1=&FloatFromCurve(FloatCurve=DataInterface() {propertyValues {Curve {keys [
         {interpMode=RCIM_Cubic, leaveTangent=-3.7537, tangentMode=RCTM_Break, time=0, value=1}
         {arriveTangent=0.1636, tangentMode=RCTM_User, time=0.6, value=0}
        ]}}})
    Index 0 Param 2=0.7
    Module.Index0Param0Write=true
    Module.Index0Param1Write=true
    Module.Index0Param2Write=true
    Module.Index0Param3Write=true
   )
  ]
  rendererClasses [&NiagaraSpriteRendererProperties]
  renderers [{rendererClass=&NiagaraSpriteRendererProperties}]
 }
 Trail {
  emitterSpawnScript [&FN_RibbonIDOffsetUtility()]
  emitterUpdateScript [
   &EmitterState()
   &FN_SpawnAnimTrail()
  ]
  particleSpawnScript [
   &FN_InitializeAnimTrailBoneIndecies()
   &FN_InitializeAnimTrail(AnimTrailWidth=1.3, Color=<rgba 1 1 1 0.5>, Lifetime=0.25, SkeletalMesh=$System.Skeletal Mesh)
   &AddVelocity(Velocity=$System.DEBUGVELSPEED) {enabled=false}
  ]
  particleUpdateScript [
   &SolveForcesAndVelocity(Clamp Velocity=false, Limit Acceleration=false)
   &DynamicMaterialParameters(
    Index 0 Param 1=1
    Index 0 Param 2=0.6
    Index 0 Param 3=1
    Index 0 Param 4=1
    Module.Param0WriteEnabled=true
    Module.Param1WriteEnabled=true
    Module.Param2WriteEnabled=true
    Module.Param3WriteEnabled=true
   )
   &ScaleColor(
    Linear Color Curve=DataInterface() {propertyValues {
      AlphaCurve {keys [<at 0.0027 1>,<at 0.2142 0.2>,<at 0.5377 0>]}
      BlueCurve {keys [<at 0.0127 0.0595>,<at 0.1128 0.0086>,<at 0.2893 0.8958>]}
      GreenCurve {keys [<at 0.0127 0.0203>,<at 0.1128 0>,<at 0.2893 0>]}
      RedCurve {keys [<at 0.0127 1>,<at 0.1128 0.2232>,<at 0.2893 0.2341>]}
     }}
    Scale Mode=::NewEnumerator2
   )
   &FN_UpdateAnimTrail(Locked Particle RibbonWidth=1, SkeletalMesh=$System.Skeletal Mesh) {module=FN_UpdateAnimTrail001}
  ]
  rendererClasses [&NiagaraRibbonRendererProperties]
  renderers [
   {rendererClass=&NiagaraRibbonRendererProperties}
   {rendererClass=&NiagaraRibbonRendererProperties, rendererIndex=1}
  ]
 }
}
summary {
 systemName=NS_Pickaxe_WeaveHarbor_Trail
 emitters [
  {emitterName=Trail, rendererClasses [&NiagaraRibbonRendererProperties]}
  {emitterName=MainLightning, rendererClasses [&NiagaraRibbonRendererProperties]}
  {emitterName=SecondaryLightning, rendererClasses [&NiagaraRibbonRendererProperties]}
  {emitterName=Eeeeeeh, rendererClasses [&NiagaraRibbonRendererProperties]}
  {bEnabled=false, emitterName=Smokes, rendererClasses [&NiagaraSpriteRendererProperties]}
 ]
 userVariables [
  User.RibbonIDOffset {type=NiagaraInt32}
  User.SkeletalMesh {
   type=NiagaraDataInterfaceSkeletalMesh
   defaultValue {
    struct=NiagaraExt_VariableValue_DataInterface
    value {dataInterface=&self:NiagaraDataInterfaceSkeletalMesh_0}
   }
  }
 ]
}
