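# Material function for the WeaveHarbor pickaxe that adds a timed reactive look, then fades back to the original material.
# Time since the BP_GetTime parameter, over ReactivityDuration, drives a smoothstepped blend between original and reactive attributes; a debug static switch forces full effect.
# An EffectMask, shaped from the input mask's green channel with red-channel selection of a minimum-remapped variant, controls reactive color, roughness and emissive.
# A sine oscillates emissive brightness between minimum and maximum parameters.
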MaterialFunction /BRCosmetics/Weapons/FORT_Melee/Pickaxe_WeaveHarbor/Materials/MF_Pickaxe_WeaveHarbor
graph {nodes [
  n1 = Constant() {r=1}
  n2 = Divide(
   A=Subtract(A=Time(), B=ScalarParameter() {defaultValue=-9999, group=Reactivity, parameterName=BP_GetTime, sortPriority=0})
   B=ScalarParameter() {defaultValue=5.5, group=Reactivity, parameterName=ReactivityDuration}
  )
  n3 = SmoothStep(Value=OneMinus(None=Saturate(None=@n2))) {constMax=0.3}
  n4 = GetMaterialAttributes(None=FunctionInput() {inputName=MaterialAttributes, inputType=FunctionInput_MaterialAttributes}) {desc=OG Material, attributeGetTypes [#1, #2]}
  n5 = VectorParameter() {group=Reactivity, parameterName=ReactiveBaseColor, sortPriority=10, defaultValue=<rgba 2 0.0162 0 1>}
  n6 = NamedRerouteUsage()
  n7 = VectorParameter() {group=Reactivity, parameterName=ReactiveEmissiveColor, sortPriority=11, defaultValue=<rgba 2 0.0162 0 1>}
  n8 = ScalarParameter() {defaultValue=0.5, group=Reactivity, parameterName=ReactiveMinBrightness, sortPriority=11}
  n9 = ScalarParameter() {defaultValue=1, group=Reactivity, parameterName=ReactiveMaxBrightness, sortPriority=11}
  n10 = Multiply(A=Time(), B=ScalarParameter() {defaultValue=1, group=Reactivity, parameterName=ReactiveOscilationFrequency, sortPriority=11})
  n11 = Add(A=Multiply(A=Sine(None=@n10)) {constB=0.5}) {constB=0.5}
  n12 = SetMaterialAttributes!(
   MaterialAttributes=@n4.MaterialAttributes
   BaseColor=LinearInterpolate(A=@n4.BaseColor, B=@n5.RGB, Alpha=NamedRerouteUsage())
   Roughness=LinearInterpolate(A=@n4.Roughness, Alpha=NamedRerouteUsage()) {constB=0.42}
   EmissiveColor=Multiply(A=Multiply(A=@n6, B=@n7.RGB), B=LinearInterpolate(A=@n8, B=@n9, Alpha=@n11))
  ) {
   desc=New Material
   attributeSetTypes [#1, #2, #3]
  }
  FunctionOutput(None=BlendMaterialAttributes(
    A=@n4.MaterialAttributes
    B=@n12
    Alpha=StaticSwitchParameter(True=@n1, False=@n3) {group=Reactivity, parameterName=DebugReactiveLook, sortPriority=1}
   ) {desc="Blend between the OG material and \r\nthe reactive one."})
  n13 = FunctionInput() {inputName=Mask, sortPriority=10}
  n14 = Power(Base=Add(A=ComponentMask(None=@n13) {g=true}) {constB=0.0001}) {constExponent=0.3}
  n15 = ScalarParameter() {group=Reactivity, parameterName=ReactivityMask_Min}
  NamedRerouteDeclaration(None=LinearInterpolate(A=@n14, B=LinearInterpolate(A=@n15, Alpha=@n14), Alpha=ComponentMask(None=@n13) {r=true})) {name=EffectMask, nodeColor=<rgba 0 1 0.3438 1>}
 ]}
