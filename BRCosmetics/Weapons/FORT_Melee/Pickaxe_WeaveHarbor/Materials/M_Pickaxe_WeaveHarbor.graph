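# Surface material for the WeaveHarbor pickaxe cosmetic, built on the 2023 character parent material functions using material attributes.
# The PreFX parent's attributes pass through MF_Pickaxe_WeaveHarbor, masked by an FX texture that is marked as a placeholder, then into the PostFX parent.
# Shading model comes from the expression, and usage covers skeletal, cloth, Nanite, mesh deformer and Niagara mesh particles.
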Material /BRCosmetics/Weapons/FORT_Melee/Pickaxe_WeaveHarbor/Materials/M_Pickaxe_WeaveHarbor
graph {
 properties {
  bUseMaterialAttributes=true
  bUsedWithClothing=true
  bUsedWithMeshDeformer=true
  bUsedWithNanite=true
  bUsedWithNiagaraMeshParticles=true
  bUsedWithSkeletalMesh=true
  pixelDepthOffsetMode=PDOM_Legacy
  shadingModel=MSM_FromMaterialExpression
  usedShadingModels=Default Lit | Subsurface | Clear Coat | Subsurface Profile | Cloth
 }
 nodes [
  n1 = MaterialFunctionCall() {materialFunction=&MF_CharacterParent_2023_PreFX}
  n2 = TextureSample() {
   desc="Placeholder until the actual FX texture\r\nis updated on the project"
   samplerSource=SSM_Wrap_WorldGroupSettings
   texture=/BRCosmetics/Weapons/FORT_Melee/Pickaxe_WeaveHarbor/Textures/T_Pickaxe_WeaveHarbor_FX_Placeholder
  }
  MaterialFunctionCall(
   Material=MaterialFunctionCall!(MaterialAttributes=@n1.Result, Mask=@n2.RGB) {materialFunction=&/MF_Pickaxe_WeaveHarbor}.Result
   M=@n1.M
   Pre Skinned Local Position=@n1.M
   Distance Blend=@n1.M
   Fresnel=@n1.M
  ) {materialFunction=&MF_CharacterParent_2023_PostFX}
 ]
}
