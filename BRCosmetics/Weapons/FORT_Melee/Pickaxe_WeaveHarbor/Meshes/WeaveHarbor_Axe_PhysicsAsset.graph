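# Physics asset for the WeaveHarbor pickaxe, a two-handed melee weapon cosmetic.
# It defines a single capsule body on the Melee_TwoHanded_Handle bone, offset and rotated along that bone.
# There are no constraints, so the axe has one collision shape and no jointed parts.
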PhysicsAsset /BRCosmetics/Weapons/FORT_Melee/Pickaxe_WeaveHarbor/Meshes/WeaveHarbor_Axe_PhysicsAsset
bodies {Melee_TwoHanded_Handle [{shapeType=Capsule, length=72.6563, radius=25.4464, center=<0.7977 -0.8964 29.3488>, rotation=<rot -0.0368 -90 0.0285>}]}
constraints []
